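# Animation Blueprint for the TutSoothe backpack, adding procedural secondary motion to its charm, legs, arms, ears and earrings.
# The input pose gets a small additive offset on the pack bone, then is cached to feed two serial AnimDynamics chains using angular springs and limits.
# bIsInFrontEnd selects between these separately tuned chains, one for front end and one for elsewhere.
# Ears and earrings simulate root-relative; the event graph has no connected logic.
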AnimBlueprint /BRCosmetics/Accessories/FORT_Backpacks/Backpack_F_MED_TutSoothe/Meshes/F_MED_TutSoothe_Pack_AnimBP
target_skeleton=&/F_MED_TutSoothe_Pack_Skeleton
graphs [
 AnimGraph {nodes [
   n1 = Anim Dynamics - Bone: dyn_charm_01(ComponentPose=Local To Component(LocalPose=Use cached pose 'StartCache'()), Alpha=1, ExternalForce="0,0,0") {
    angularDampingOverride=0.999
    angularSpringConstant=2000
    bAngularSpring=true
    bOverrideAngularDamping=true
    bUseGravityOverride=true
    gravityScale=0.35
    boundBone {boneName=dyn_charm_01}
    componentLinearAccScale=<1 1 1>
    componentLinearVelScale=<1 1 1>
    gravityOverride {z=-980}
    physicsBodyDefinitions [{
      boundBone {boneName=dyn_charm_01}
      boxExtents=<20 30 30>
      constraintSetup {angularLimitsMax=<95 45 15>, angularLimitsMin=<-90 -15 -15>, angularTarget=<1 0 0>}
      localJointOffset=<4 0 0>
     }]
    simSpaceSettings {maxAngularAcceleration=100, maxAngularVelocity=400, simSpaceAngularAlpha=1}
   }
   n2 = Anim Dynamics - Bone: dyn_charm_02(ComponentPose=@n1, Alpha=1, ExternalForce="0,0,0") {
    angularDampingOverride=0.999
    angularSpringConstant=2000
    bAngularSpring=true
    bOverrideAngularDamping=true
    bUseGravityOverride=true
    gravityScale=0.35
    boundBone {boneName=dyn_charm_02}
    componentLinearAccScale=<1 1 1>
    componentLinearVelScale=<1 1 1>
    gravityOverride {z=-980}
    physicsBodyDefinitions [{
      boundBone {boneName=dyn_charm_02}
      boxExtents=<20 30 30>
      constraintSetup {angularLimitsMax=<95 45 15>, angularLimitsMin=<-90 -15 -15>, angularTarget=<1 0 0>}
      localJointOffset=<4 0 0>
     }]
    simSpaceSettings {maxAngularAcceleration=100, maxAngularVelocity=400, simSpaceAngularAlpha=1}
   }
   n3 = Anim Dynamics - Bone: dyn_charm_03(ComponentPose=@n2, Alpha=1, ExternalForce="0,0,0") {
    angularDampingOverride=0.999
    angularSpringConstant=2000
    bAngularSpring=true
    bOverrideAngularDamping=true
    bUseGravityOverride=true
    gravityScale=0.35
    boundBone {boneName=dyn_charm_03}
    componentLinearAccScale=<1 1 1>
    componentLinearVelScale=<1 1 1>
    gravityOverride {z=-980}
    physicsBodyDefinitions [{
      boundBone {boneName=dyn_charm_03}
      boxExtents=<20 30 30>
      constraintSetup {angularLimitsMax=<95 45 15>, angularLimitsMin=<-90 -15 -15>, angularTarget=<1 0 0>}
      localJointOffset=<4 0 0>
     }]
    simSpaceSettings {maxAngularAcceleration=100, maxAngularVelocity=400, simSpaceAngularAlpha=1}
   }
   n4 = Anim Dynamics - Bone: dyn_charm_04(ComponentPose=@n3, Alpha=1, ExternalForce="0,0,0") {
    angularDampingOverride=0.999
    angularSpringConstant=2000
    bAngularSpring=true
    bOverrideAngularDamping=true
    bUseGravityOverride=true
    gravityScale=0.35
    boundBone {boneName=dyn_charm_04}
    componentLinearAccScale=<1 1 1>
    componentLinearVelScale=<1 1 1>
    gravityOverride {z=-980}
    physicsBodyDefinitions [{
      boundBone {boneName=dyn_charm_04}
      boxExtents=<20 30 30>
      constraintSetup {angularLimitsMax=<95 45 15>, angularLimitsMin=<-90 -15 -15>, angularTarget=<1 0 0>}
      localJointOffset=<4 0 0>
     }]
    simSpaceSettings {maxAngularAcceleration=100, maxAngularVelocity=400, simSpaceAngularAlpha=1}
   }
   n5 = Anim Dynamics - Bone: dyn_leg_01_l(ComponentPose=@n4, Alpha=1, ExternalForce="0,0,0") {
    angularDampingOverride=1
    angularSpringConstant=3000
    bAngularSpring=true
    bOverrideAngularDamping=true
    bUseGravityOverride=true
    gravityScale=0.35
    boundBone {boneName=dyn_leg_01_l}
    componentAppliedLinearAccClamp=<190 190 690>
    componentLinearAccScale=<0.65 0.65 0.65>
    componentLinearVelScale=<0.75 0.75 0.75>
    gravityOverride {z=-980}
    physicsBodyDefinitions [{
      boundBone {boneName=dyn_leg_01_l}
      boxExtents=<9 9 8>
      constraintSetup {angularLimitsMax=<5 5 5>, angularLimitsMin=<-4 -4 -4>, angularTarget=<1 0 0>}
      localJointOffset=<4 0 0>
     }]
    simSpaceSettings {maxAngularAcceleration=150, maxAngularVelocity=450, simSpaceAngularAlpha=1}
   }
   n6 = Anim Dynamics - Bone: dyn_leg_01_r(ComponentPose=@n5, Alpha=1, ExternalForce="0,0,0") {
    angularDampingOverride=1
    angularSpringConstant=3000
    bAngularSpring=true
    bOverrideAngularDamping=true
    bUseGravityOverride=true
    gravityScale=0.35
    boundBone {boneName=dyn_leg_01_r}
    componentAppliedLinearAccClamp=<190 190 690>
    componentLinearAccScale=<0.65 0.65 0.65>
    componentLinearVelScale=<0.75 0.75 0.75>
    gravityOverride {z=-980}
    physicsBodyDefinitions [{
      boundBone {boneName=dyn_leg_01_r}
      boxExtents=<9 9 8>
      constraintSetup {angularLimitsMax=<5 5 5>, angularLimitsMin=<-4 -4 -4>, angularTarget=<1 0 0>}
      localJointOffset=<4 0 0>
     }]
    simSpaceSettings {maxAngularAcceleration=150, maxAngularVelocity=450, simSpaceAngularAlpha=1}
   }
   n7 = Anim Dynamics - Bone: dyn_arm_01_r(ComponentPose=@n6, Alpha=1, ExternalForce="0,0,0") {
    angularDampingOverride=1
    angularSpringConstant=3000
    bAngularSpring=true
    bOverrideAngularDamping=true
    bUseGravityOverride=true
    gravityScale=0.35
    boundBone {boneName=dyn_arm_01_r}
    componentAppliedLinearAccClamp=<1000 1000 1000>
    componentLinearAccScale=<1 1 1>
    componentLinearVelScale=<1 1 1>
    gravityOverride {z=-680}
    physicsBodyDefinitions [{
      boundBone {boneName=dyn_arm_01_r}
      boxExtents=<15 15 15>
      constraintSetup {angularLimitsMax=<35 5 30>, angularLimitsMin=<-30 -4 -28>, angularTarget=<1 0 0>}
      localJointOffset=<6 0 0>
     }]
    simSpaceSettings {maxAngularAcceleration=30, maxAngularVelocity=450, simSpaceAngularAlpha=1}
   }
   n8 = Anim Dynamics - Bone: dyn_arm_01_l(ComponentPose=@n7, Alpha=1, ExternalForce="0,0,0") {
    angularDampingOverride=1
    angularSpringConstant=3000
    bAngularSpring=true
    bOverrideAngularDamping=true
    bUseGravityOverride=true
    gravityScale=0.35
    boundBone {boneName=dyn_arm_01_l}
    componentAppliedLinearAccClamp=<1000 1000 1000>
    componentLinearAccScale=<1 1 1>
    componentLinearVelScale=<1 1 0.75>
    gravityOverride {z=-680}
    physicsBodyDefinitions [{
      boundBone {boneName=dyn_arm_01_l}
      boxExtents=<15 15 15>
      constraintSetup {angularLimitsMax=<35 5 30>, angularLimitsMin=<-30 -4 -28>, angularTarget=<1 0 0>}
      localJointOffset=<6 0 0>
     }]
    simSpaceSettings {maxAngularAcceleration=30, maxAngularVelocity=450, simSpaceAngularAlpha=1}
   }
   n9 = Anim Dynamics - Bone: dyn_ear_01_r(ComponentPose=@n8, Alpha=1, ExternalForce="0,0,0") {
    angularDampingOverride=1
    angularSpringConstant=20000
    bAngularSpring=true
    bOverrideAngularDamping=true
    bUseGravityOverride=true
    gravityScale=0.35
    simulationSpace=RootRelative
    boundBone {boneName=dyn_ear_01_r}
    componentLinearAccScale=<1 1 1>
    componentLinearVelScale=<1 1 1>
    gravityOverride {z=-980}
    physicsBodyDefinitions [{
      boundBone {boneName=dyn_ear_01_r}
      boxExtents=<45 28 28>
      constraintSetup {angularLimitsMax=<0 5 20>, angularLimitsMin=<0 -5 -13>, angularTarget=<1 0 0>}
      localJointOffset=<4 0 0>
     }]
    simSpaceSettings {maxAngularAcceleration=1000, maxAngularVelocity=5000, simSpaceAngularAlpha=0.1, externalAngularVelocity=<0.5 0.5 0.5>}
   }
   n10 = Anim Dynamics - Bone: dyn_ear_02_r(ComponentPose=@n9, Alpha=1, ExternalForce="0,0,0") {
    angularDampingOverride=1
    angularSpringConstant=20000
    bAngularSpring=true
    bOverrideAngularDamping=true
    bUseGravityOverride=true
    gravityScale=0.35
    simulationSpace=RootRelative
    boundBone {boneName=dyn_ear_02_r}
    componentLinearAccScale=<1 1 1>
    componentLinearVelScale=<1 1 1>
    gravityOverride {z=-980}
    physicsBodyDefinitions [{
      boundBone {boneName=dyn_ear_02_r}
      boxExtents=<45 28 28>
      constraintSetup {angularLimitsMax=<0 5 20>, angularLimitsMin=<0 -5 -13>, angularTarget=<1 0 0>}
      localJointOffset=<4 0 0>
     }]
    simSpaceSettings {maxAngularAcceleration=1000, maxAngularVelocity=5000, simSpaceAngularAlpha=0.1, externalAngularVelocity=<0.5 0.5 0.5>}
   }
   n11 = Anim Dynamics - Bone: dyn_ear_01_l(ComponentPose=@n10, Alpha=1, ExternalForce="0,0,0") {
    angularDampingOverride=1
    angularSpringConstant=20000
    bAngularSpring=true
    bOverrideAngularDamping=true
    bUseGravityOverride=true
    gravityScale=0.35
    simulationSpace=RootRelative
    boundBone {boneName=dyn_ear_01_l}
    componentLinearAccScale=<1 1 1>
    componentLinearVelScale=<1 1 1>
    gravityOverride {z=-980}
    physicsBodyDefinitions [{
      boundBone {boneName=dyn_ear_01_l}
      boxExtents=<40 30 30>
      constraintSetup {angularLimitsMax=<0 5 13>, angularLimitsMin=<0 -5 -13>, angularTarget=<1 0 0>}
      localJointOffset=<4 0 0>
     }]
    simSpaceSettings {maxAngularAcceleration=1000, maxAngularVelocity=5000, simSpaceAngularAlpha=0.1, externalAngularVelocity=<0.5 0.5 0.5>}
   }
   n12 = Anim Dynamics - Bone: dyn_ear_02_l(ComponentPose=@n11, Alpha=1, ExternalForce="0,0,0") {
    angularDampingOverride=1
    angularSpringConstant=20000
    bAngularSpring=true
    bOverrideAngularDamping=true
    bUseGravityOverride=true
    gravityScale=0.35
    simulationSpace=RootRelative
    boundBone {boneName=dyn_ear_02_l}
    componentLinearAccScale=<1 1 1>
    componentLinearVelScale=<1 1 1>
    gravityOverride {z=-980}
    physicsBodyDefinitions [{
      boundBone {boneName=dyn_ear_02_l}
      boxExtents=<40 30 30>
      constraintSetup {angularLimitsMax=<0 5 16>, angularLimitsMin=<0 -5 -16>, angularTarget=<1 0 0>}
      localJointOffset=<4 0 0>
     }]
    simSpaceSettings {maxAngularAcceleration=1000, maxAngularVelocity=5000, simSpaceAngularAlpha=0.1, externalAngularVelocity=<0.5 0.5 0.5>}
   }
   n13 = Anim Dynamics - Bone: dyn_earing_02(ComponentPose=@n12, Alpha=1, ExternalForce="0,0,0") {
    angularDampingOverride=1
    angularSpringConstant=20000
    bAngularSpring=true
    bOverrideAngularDamping=true
    bUseGravityOverride=true
    gravityScale=0.35
    simulationSpace=RootRelative
    boundBone {boneName=dyn_earing_02}
    componentLinearAccScale=<1 1 1>
    componentLinearVelScale=<1 1 1>
    gravityOverride {z=-980}
    physicsBodyDefinitions [{
      boundBone {boneName=dyn_earing_02}
      boxExtents=<20 30 30>
      constraintSetup {angularLimitsMax=<0 29 0>, angularLimitsMin=<0 -28 0>, angularTarget=<1 0 0>}
      localJointOffset=<4 0 0>
     }]
   }
   n14 = Anim Dynamics - Bone: dyn_earing_01(ComponentPose=@n13, Alpha=1, ExternalForce="0,0,0") {
    angularDampingOverride=1
    angularSpringConstant=10000
    bAngularSpring=true
    bOverrideAngularDamping=true
    bUseGravityOverride=true
    gravityScale=0.35
    simulationSpace=RootRelative
    boundBone {boneName=dyn_earing_01}
    componentLinearAccScale=<1 1 1>
    componentLinearVelScale=<1 1 1>
    gravityOverride {z=-980}
    physicsBodyDefinitions [{
      boundBone {boneName=dyn_earing_01}
      boxExtents=<10 20 20>
      constraintSetup {angularLimitsMax=<0 45 0>, angularLimitsMin=<0 -30 0>, angularTarget=<1 0 0>}
      localJointOffset=<4 0 0>
     }]
    simSpaceSettings {maxAngularAcceleration=1000, maxAngularVelocity=5000, simSpaceAngularAlpha=0.1}
   }
   n15 = Anim Dynamics - Bone: dyn_charm_01(ComponentPose=Local To Component(LocalPose=Use cached pose 'StartCache'()), Alpha=1, ExternalForce="0,0,0") {
    angularDampingOverride=0.999
    angularSpringConstant=100
    bAngularSpring=true
    bOverrideAngularDamping=true
    bUseGravityOverride=true
    gravityScale=0.35
    boundBone {boneName=dyn_charm_01}
    componentAppliedLinearAccClamp=<2500 2500 2500>
    componentLinearAccScale=<0.6 0.6 0.6>
    componentLinearVelScale=<0.5 0.5 0.5>
    gravityOverride {z=-980}
    physicsBodyDefinitions [{
      boundBone {boneName=dyn_charm_01}
      boxExtents=<10 20 20>
      constraintSetup {angularLimitsMax=<70 40 10>, angularLimitsMin=<-80 -15 -15>, angularTarget=<1 0 0>}
      localJointOffset=<4 0 0>
     }]
    simSpaceSettings {maxAngularAcceleration=50, maxAngularVelocity=400, simSpaceAngularAlpha=1}
   }
   n16 = Anim Dynamics - Bone: dyn_charm_02(ComponentPose=@n15, Alpha=1, ExternalForce="0,0,0") {
    angularDampingOverride=0.999
    angularSpringConstant=100
    bAngularSpring=true
    bOverrideAngularDamping=true
    bUseGravityOverride=true
    gravityScale=0.35
    boundBone {boneName=dyn_charm_02}
    componentAppliedLinearAccClamp=<2500 2500 2500>
    componentLinearAccScale=<0.6 0.6 0.6>
    componentLinearVelScale=<0.5 0.5 0.5>
    gravityOverride {z=-980}
    physicsBodyDefinitions [{
      boundBone {boneName=dyn_charm_02}
      boxExtents=<10 20 20>
      constraintSetup {angularLimitsMax=<70 40 10>, angularLimitsMin=<-80 -15 -15>, angularTarget=<1 0 0>}
      localJointOffset=<4 0 0>
     }]
    simSpaceSettings {maxAngularAcceleration=50, maxAngularVelocity=400, simSpaceAngularAlpha=1}
   }
   n17 = Anim Dynamics - Bone: dyn_charm_03(ComponentPose=@n16, Alpha=1, ExternalForce="0,0,0") {
    angularDampingOverride=0.999
    angularSpringConstant=100
    bAngularSpring=true
    bOverrideAngularDamping=true
    bUseGravityOverride=true
    gravityScale=0.35
    boundBone {boneName=dyn_charm_03}
    componentAppliedLinearAccClamp=<2500 2500 2500>
    componentLinearAccScale=<0.6 0.6 0.6>
    componentLinearVelScale=<0.5 0.5 0.5>
    gravityOverride {z=-980}
    physicsBodyDefinitions [{
      boundBone {boneName=dyn_charm_03}
      boxExtents=<10 20 20>
      constraintSetup {angularLimitsMax=<70 40 10>, angularLimitsMin=<-80 -15 -15>, angularTarget=<1 0 0>}
      localJointOffset=<4 0 0>
     }]
    simSpaceSettings {maxAngularAcceleration=50, maxAngularVelocity=400, simSpaceAngularAlpha=1}
   }
   n18 = Anim Dynamics - Bone: dyn_charm_04(ComponentPose=@n17, Alpha=1, ExternalForce="0,0,0") {
    angularDampingOverride=0.999
    angularSpringConstant=100
    bAngularSpring=true
    bOverrideAngularDamping=true
    bUseGravityOverride=true
    gravityScale=0.35
    boundBone {boneName=dyn_charm_04}
    componentAppliedLinearAccClamp=<2500 2500 2500>
    componentLinearAccScale=<0.6 0.6 0.6>
    componentLinearVelScale=<0.5 0.5 0.5>
    gravityOverride {z=-980}
    physicsBodyDefinitions [{
      boundBone {boneName=dyn_charm_04}
      boxExtents=<10 20 20>
      constraintSetup {angularLimitsMax=<70 40 10>, angularLimitsMin=<-80 -15 -15>, angularTarget=<1 0 0>}
      localJointOffset=<4 0 0>
     }]
    simSpaceSettings {maxAngularAcceleration=50, maxAngularVelocity=400, simSpaceAngularAlpha=1}
   }
   n19 = Anim Dynamics - Bone: dyn_leg_01_l(ComponentPose=@n18, Alpha=1, ExternalForce="0,0,0") {
    angularDampingOverride=0.999
    angularSpringConstant=3000
    bAngularSpring=true
    bOverrideAngularDamping=true
    bUseGravityOverride=true
    gravityScale=0.35
    boundBone {boneName=dyn_leg_01_l}
    componentAppliedLinearAccClamp=<190 190 690>
    componentLinearAccScale=<0.65 0.65 0.65>
    componentLinearVelScale=<0.75 0.75 0.75>
    gravityOverride {z=-980}
    physicsBodyDefinitions [{
      boundBone {boneName=dyn_leg_01_l}
      boxExtents=<25 25 25>
      constraintSetup {angularLimitsMax=<5 5 5>, angularLimitsMin=<-4 -4 -4>, angularTarget=<1 0 0>}
      localJointOffset=<4 0 0>
     }]
    simSpaceSettings {maxAngularAcceleration=150, maxAngularVelocity=450, simSpaceAngularAlpha=1}
   }
   n20 = Anim Dynamics - Bone: dyn_leg_01_r(ComponentPose=@n19, Alpha=1, ExternalForce="0,0,0") {
    angularDampingOverride=0.999
    angularSpringConstant=3000
    bAngularSpring=true
    bOverrideAngularDamping=true
    bUseGravityOverride=true
    gravityScale=0.35
    boundBone {boneName=dyn_leg_01_r}
    componentAppliedLinearAccClamp=<190 190 690>
    componentLinearAccScale=<0.65 0.65 0.65>
    componentLinearVelScale=<0.75 0.75 0.75>
    gravityOverride {z=-980}
    physicsBodyDefinitions [{
      boundBone {boneName=dyn_leg_01_r}
      boxExtents=<25 25 25>
      constraintSetup {angularLimitsMax=<5 5 5>, angularLimitsMin=<-4 -4 -4>, angularTarget=<1 0 0>}
      localJointOffset=<4 0 0>
     }]
    simSpaceSettings {maxAngularAcceleration=150, maxAngularVelocity=450, simSpaceAngularAlpha=1}
   }
   n21 = Anim Dynamics - Bone: dyn_arm_01_r(ComponentPose=@n20, Alpha=1, ExternalForce="0,0,0") {
    angularDampingOverride=0.9
    angularSpringConstant=13600
    bAngularSpring=true
    bOverrideAngularDamping=true
    bUseGravityOverride=true
    gravityScale=0.35
    boundBone {boneName=dyn_arm_01_r}
    componentAppliedLinearAccClamp=<1000 1000 1000>
    componentLinearAccScale=<1 1 1>
    componentLinearVelScale=<1 1 1>
    gravityOverride {z=-680}
    physicsBodyDefinitions [{
      boundBone {boneName=dyn_arm_01_r}
      boxExtents=<60 60 60>
      constraintSetup {angularLimitsMax=<6 17 0>, angularLimitsMin=<-12 -8 0>, angularTarget=<1 0 0>}
      localJointOffset=<6 0 0>
     }]
    simSpaceSettings {maxAngularAcceleration=325, maxAngularVelocity=400, simSpaceAngularAlpha=1, externalAngularVelocity {y=2}}
   }
   n22 = Anim Dynamics - Bone: dyn_arm_01_l(ComponentPose=@n21, Alpha=1, ExternalForce="0,0,0") {
    angularDampingOverride=0.91
    angularSpringConstant=13500
    bAngularSpring=true
    bOverrideAngularDamping=true
    bUseGravityOverride=true
    gravityScale=0.35
    boundBone {boneName=dyn_arm_01_l}
    componentAppliedLinearAccClamp=<1000 1000 1000>
    componentLinearAccScale=<1 1 1>
    componentLinearVelScale=<1 1 0.75>
    gravityOverride {z=-680}
    physicsBodyDefinitions [{
      boundBone {boneName=dyn_arm_01_l}
      boxExtents=<60 60 60>
      constraintSetup {angularLimitsMax=<5 18 0>, angularLimitsMin=<-14 -9 0>, angularTarget=<1 0 0>}
      localJointOffset=<6 0 0>
     }]
    simSpaceSettings {maxAngularAcceleration=330, maxAngularVelocity=400, simSpaceAngularAlpha=1, externalAngularVelocity {y=2}}
   }
   n23 = Anim Dynamics - Bone: dyn_ear_01_r(ComponentPose=@n22, Alpha=1, ExternalForce="0,0,0") {
    angularDampingOverride=0.99
    angularSpringConstant=10000
    bAngularSpring=true
    bOverrideAngularDamping=true
    bUseGravityOverride=true
    gravityScale=0.35
    simulationSpace=RootRelative
    boundBone {boneName=dyn_ear_01_r}
    componentLinearAccScale=<1 1 1>
    componentLinearVelScale=<1 1 1>
    gravityOverride {z=-980}
    physicsBodyDefinitions [{
      boundBone {boneName=dyn_ear_01_r}
      boxExtents=<60 57 57>
      constraintSetup {angularLimitsMax=<0 5 8>, angularLimitsMin=<0 -5 -10>, angularTarget=<1 0 0>}
      localJointOffset=<4 0 0>
     }]
    simSpaceSettings {maxAngularAcceleration=90, maxAngularVelocity=350, simSpaceAngularAlpha=1, externalAngularVelocity=<0.5 0.5 0.5>}
   }
   n24 = Anim Dynamics - Bone: dyn_ear_02_r(ComponentPose=@n23, Alpha=1, ExternalForce="0,0,0") {
    angularDampingOverride=0.992
    angularSpringConstant=10000
    bAngularSpring=true
    bOverrideAngularDamping=true
    bUseGravityOverride=true
    gravityScale=0.35
    simulationSpace=RootRelative
    boundBone {boneName=dyn_ear_02_r}
    componentLinearAccScale=<1 1 1>
    componentLinearVelScale=<1 1 1>
    gravityOverride {z=-980}
    physicsBodyDefinitions [{
      boundBone {boneName=dyn_ear_02_r}
      boxExtents=<60 57 57>
      constraintSetup {angularLimitsMax=<0 5 13>, angularLimitsMin=<0 -5 -15>, angularTarget=<1 0 0>}
      localJointOffset=<4 0 0>
     }]
    simSpaceSettings {maxAngularAcceleration=90, maxAngularVelocity=350, simSpaceAngularAlpha=1, externalAngularVelocity=<0.5 0.5 0.5>}
   }
   n25 = Anim Dynamics - Bone: dyn_ear_01_l(ComponentPose=@n24, Alpha=1, ExternalForce="0,0,0") {
    angularDampingOverride=0.99
    angularSpringConstant=10000
    bAngularSpring=true
    bOverrideAngularDamping=true
    bUseGravityOverride=true
    gravityScale=0.35
    simulationSpace=RootRelative
    boundBone {boneName=dyn_ear_01_l}
    componentLinearAccScale=<1 1 1>
    componentLinearVelScale=<1 1 1>
    gravityOverride {z=-980}
    physicsBodyDefinitions [{
      boundBone {boneName=dyn_ear_01_l}
      boxExtents=<58 58 58>
      constraintSetup {angularLimitsMax=<0 5 9>, angularLimitsMin=<0 -5 -11>, angularTarget=<1 0 0>}
      localJointOffset=<4 0 0>
     }]
    simSpaceSettings {maxAngularAcceleration=80, maxAngularVelocity=350, simSpaceAngularAlpha=1, externalAngularVelocity=<0.5 0.5 0.5>}
   }
   n26 = Anim Dynamics - Bone: dyn_ear_02_l(ComponentPose=@n25, Alpha=1, ExternalForce="0,0,0") {
    angularDampingOverride=0.9925
    angularSpringConstant=10000
    bAngularSpring=true
    bOverrideAngularDamping=true
    bUseGravityOverride=true
    gravityScale=0.35
    simulationSpace=RootRelative
    boundBone {boneName=dyn_ear_02_l}
    componentLinearAccScale=<1 1 1>
    componentLinearVelScale=<1 1 1>
    gravityOverride {z=-980}
    physicsBodyDefinitions [{
      boundBone {boneName=dyn_ear_02_l}
      boxExtents=<58 58 58>
      constraintSetup {angularLimitsMax=<0 5 11>, angularLimitsMin=<0 -5 -15>, angularTarget=<1 0 0>}
      localJointOffset=<4 0 0>
     }]
    simSpaceSettings {maxAngularAcceleration=80, maxAngularVelocity=350, simSpaceAngularAlpha=1, externalAngularVelocity=<0.5 0.5 0.5>}
   }
   n27 = Anim Dynamics - Bone: dyn_earing_02(ComponentPose=@n26, Alpha=1, ExternalForce="0,0,0") {
    angularDampingOverride=0.999
    angularSpringConstant=10000
    bAngularSpring=true
    bOverrideAngularDamping=true
    bUseGravityOverride=true
    gravityScale=0.35
    simulationSpace=RootRelative
    boundBone {boneName=dyn_earing_02}
    componentLinearAccScale=<1 1 1>
    componentLinearVelScale=<1 1 1>
    gravityOverride {z=-980}
    physicsBodyDefinitions [{
      boundBone {boneName=dyn_earing_02}
      boxExtents=<45 50 50>
      constraintSetup {angularLimitsMax=<0 29 0>, angularLimitsMin=<0 -28 0>, angularTarget=<1 0 0>}
      localJointOffset=<4 0 0>
     }]
    simSpaceSettings {maxAngularAcceleration=70, maxAngularVelocity=500, simSpaceAngularAlpha=1}
   }
   n28 = Anim Dynamics - Bone: dyn_earing_01(ComponentPose=@n27, Alpha=1, ExternalForce="0,0,0") {
    angularDampingOverride=0.999
    angularSpringConstant=10000
    bAngularSpring=true
    bOverrideAngularDamping=true
    bUseGravityOverride=true
    gravityScale=0.35
    simulationSpace=RootRelative
    boundBone {boneName=dyn_earing_01}
    componentLinearAccScale=<1 1 1>
    componentLinearVelScale=<1 1 1>
    gravityOverride {z=-980}
    physicsBodyDefinitions [{
      boundBone {boneName=dyn_earing_01}
      boxExtents=<30 40 40>
      constraintSetup {angularLimitsMax=<0 45 0>, angularLimitsMin=<0 -30 0>, angularTarget=<1 0 0>}
      localJointOffset=<4 0 0>
     }]
    simSpaceSettings {maxAngularAcceleration=80, maxAngularVelocity=500, simSpaceAngularAlpha=1}
   }
   Output Pose(Result=Blend Poses by bool(BlendPose_0=Component To Local(ComponentPose=@n28), BlendPose_1=Component To Local(ComponentPose=@n14), bActiveValue=Get bIsInFrontEnd())) {layerGroup=None}
   n29 = Local To Component(LocalPose=Input Pose())
   Save cached pose 'StartCache'(Pose=Component To Local(ComponentPose="Transform (Modify) Bone - Bone: pack"(ComponentPose=@n29, Translation="0.000,-1.131,0.000", Rotation="0,0,0") {translationMode=BMM_Additive, boneToModify {boneName=pack}, translation {y=-1.131}})) {cacheName=StartCache}
  ]}
 EventGraph {nodes [
   TryGetPawnOwner()
   Event BlueprintUpdateAnimation()
  ]}
]
variables []
